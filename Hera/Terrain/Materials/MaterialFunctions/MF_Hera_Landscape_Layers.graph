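# Material function that exposes Hera's landscape layers and terrain masks as named outputs.
# It makes one call to MF_Hera_StaticMeshLandscapeLayerMatch, leaving the custom grass slope mask inputs unconnected.
# Results pass through color-coded named reroutes to outputs for four TerrainMasks quads, grass variants, desert grass noise, gravel, mud, sand, asphalt, scree and snow.
# OasisCliffMask (a default constant) and Landscape_MossyForest (no input) are reroutes that reach no output.
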MaterialFunction /Hera/Terrain/Materials/MaterialFunctions/MF_Hera_Landscape_Layers
graph {nodes [
  n1 = MaterialFunctionCall!(Grass_UseCustomSlopeMask, Grass_CustomSlopeMask) {materialFunction=/Hera/Environment/Terrain/Landscape/Materials/Layers/MF_Hera_StaticMeshLandscapeLayerMatch}
  FunctionOutput(None=NamedRerouteDeclaration(None=@n1.ChaparralGrassLayer) {name=Landscape - Chaparral Grass, nodeColor=<rgba 0.0312 0 1 1>}) {outputName=Landscape - Chaparral Grass, sortPriority=21}
  FunctionOutput(None=NamedRerouteDeclaration(None=@n1.TerrainMasks_QuadA) {name=TerrainMasks - QuadA, nodeColor=<rgba 0.5 0 1 1>}) {outputName=TerrainMasks - QuadA}
  FunctionOutput(None=NamedRerouteDeclaration(None=@n1.AsphaltLayer) {name=Landscape - Asphalt, nodeColor=<rgba 1 0.75 0 1>}) {outputName=Landscape - Asphalt, sortPriority=40}
  FunctionOutput(None=NamedRerouteDeclaration(None=@n1.ScreeLayer) {name=Landscape - Scree, nodeColor=<rgba 0.4297 0 1 1>}) {outputName=Landscape - Scree, sortPriority=45}
  FunctionOutput(None=NamedRerouteDeclaration(None=@n1.SnowLayer) {name=Landscape - Snow, nodeColor=<rgba 0 1 0.1094 1>}) {outputName=Landscape - Snow, sortPriority=50}
  FunctionOutput(None=NamedRerouteDeclaration(None=@n1.DesertGrassSimpleNoise) {name=Landscape - Desert Grass Noise, nodeColor=<rgba 0.0312 0 1 1>}) {outputName=Landscape - Desert Grass Noise, sortPriority=23}
  FunctionOutput(None=NamedRerouteDeclaration(None=@n1.TerrainMasks_QuadB) {name=TerrainMasks - QuadB, nodeColor=<rgba 0.5 0 1 1>}) {outputName=TerrainMasks - QuadB, sortPriority=1}
  FunctionOutput(None=NamedRerouteDeclaration(None=@n1.TerrainMasks_QuadC) {name=TerrainMasks - QuadC, nodeColor=<rgba 0.5 0 1 1>}) {outputName=TerrainMasks - QuadC, sortPriority=2}
  FunctionOutput(None=NamedRerouteDeclaration(None=@n1.TerrainMasks_QuadD) {name=TerrainMasks - QuadD, nodeColor=<rgba 0.5 0 1 1>}) {outputName=TerrainMasks - QuadD, sortPriority=3}
  FunctionOutput(None=NamedRerouteDeclaration(None=@n1.DesertGrassLayer) {name=Landscape - Desert Grass, nodeColor=<rgba 0.0312 0 1 1>}) {outputName=Landscape - Desert Grass, sortPriority=22}
  FunctionOutput(None=NamedRerouteDeclaration(None=@n1.GrassLayer) {name=Landscape - Grass, nodeColor=<rgba 0.0312 0 1 1>}) {outputName=Landscape - Grass, sortPriority=20}
  FunctionOutput(None=NamedRerouteDeclaration(None=@n1.GravelLayer) {name=Landscape - Gravel, nodeColor=<rgba 1 0.8662 0.519 1>}) {outputName=Landscape - Gravel, sortPriority=25}
  FunctionOutput(None=NamedRerouteDeclaration(None=@n1.MudLayer) {name=Landscape - Mud, nodeColor=<rgba 0.4427 0.1763 0.0292 1>}) {outputName=Landscape - Mud, sortPriority=30}
  FunctionOutput(None=NamedRerouteDeclaration(None=@n1.SandLayer) {name=Landscape - Sand, nodeColor=<rgba 0 0.9297 1 1>}) {outputName=Landscape - Sand, sortPriority=35}
  NamedRerouteDeclaration(None=Constant()) {name=OasisCliffMask, nodeColor=<rgba 0.5 0 1 1>}
  NamedRerouteDeclaration() {name=Landscape_MossyForest, nodeColor=<rgba 0 0.5312 1 1>}
 ]}
